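# M_StreamingPause is an unlit translucent engine material for a spinning streaming pause icon.
# It holds a constant white color and a sample of the StreamingPauseIcon texture.
# A Rotator node with negative speed drives the texture UVs, so the icon animates in the material itself.
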Material /Engine/EngineResources/M_StreamingPause
graph {
 properties {blendMode=BLEND_Translucent, pixelDepthOffsetMode=PDOM_Legacy, shadingModel=MSM_Unlit, usedShadingModels=Unlit}
 nodes [
  Constant3Vector() {desc=Color, constant=<rgba 1 1 1 0>}
  TextureSample(UVs=Rotator() {speed=-4}) {desc=Streaming pause icon, texture=&/StreamingPauseIcon}
 ]
}
